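# Unlit, translucent, two-sided glow material for Niagara sprites, used in the FX of a Vindertech umbrella cosmetic.
# Particle color RGB passes through the shared MF_FX_EmissiveAdjustments function to form the glow color.
# A radial exponential gradient with a Density parameter is narrowed by a tight SmoothStep and multiplied by particle alpha, producing a procedural, texture-free round mask.
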Material /BRCosmetics/Gadgets/Assets/VinderTech_Umbrella/Umbrella_Season26/FX/M_Umbrella_Season16_Glow
graph {
 properties {
  bUsedWithNiagaraSprites=true
  blendMode=BLEND_Translucent
  pixelDepthOffsetMode=PDOM_Legacy
  shadingModel=MSM_Unlit
  twoSided=true
  usedShadingModels=Unlit
 }
 nodes [
  n1 = ParticleColor()
  MaterialFunctionCall(Emissive=@n1.RGB) {materialFunction=&MF_FX_EmissiveAdjustments}
  n2 = ScalarParameter() {defaultValue=0.8, parameterName=Density}
  Multiply(
   A=@n1.A
   B=SmoothStep(Value=MaterialFunctionCall(Density=@n2) {materialFunction=&RadialGradientExponential}.RadialGradientExponential) {constMax=0.08, constMin=0.02}
  )
 ]
}
