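# Unlit, translucent, two-sided material for Niagara mesh particles that draws an animated two-color line pattern for the Season 26 VinderTech umbrella.
# Two SDF texture samples, one rotated, supply distance and thickness data for repeating contour lines that scroll over time with a wave offset.
# Screen-space derivatives antialias the lines.
# An FX texture masks and modulates brightness, with particle and vertex alpha also applied.
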Material /BRCosmetics/Gadgets/Assets/VinderTech_Umbrella/Umbrella_Season26/FX/M_Umbrella_Season26_Pattern
graph {
 properties {
  bUsedWithNiagaraMeshParticles=true
  blendMode=BLEND_Translucent
  pixelDepthOffsetMode=PDOM_Legacy
  shadingModel=MSM_Unlit
  twoSided=true
  usedShadingModels=Unlit
 }
 nodes [
  n1 = Clamp(None=MaterialFunctionCall(
    Input=NamedRerouteUsage()
    Input Low=Constant()
    Input High=Constant() {r=1}
    Target Low=ScalarParameter() {defaultValue=0.005, parameterName=Thickness_Low}
    Target High=ScalarParameter() {defaultValue=0.03, parameterName=Thickness_High}
   ) {materialFunction=&RemapValueRange}.Result)
  n2 = Time()
  n3 = ScalarParameter() {defaultValue=-0.05, parameterName=Speed}
  n4 = Add(A=NamedRerouteUsage(), B=Multiply(A=@n2, B=@n3))
  n5 = DDX(Value=@n4)
  n6 = DDY(Value=@n4)
  n7 = Multiply(A=Add(A=Abs(None=@n5), B=Abs(None=@n6))) {constB=0.5}
  n8 = ScalarParameter() {defaultValue=0.08, parameterName=LineDistance}
  n9 = Add(A=Divide(A=@n4, B=@n8)) {constB=0.5}
  n10 = Abs(None=Subtract(A=Frac(None=@n9)) {constB=0.5})
  n11 = OneMinus(None=SmoothStep(Min=Subtract(A=@n1, B=@n7), Max=Add(A=@n1, B=@n7), Value=Multiply(A=@n10, B=@n8)))
  n12 = Clamp(None=MaterialFunctionCall(
    Input=NamedRerouteUsage()
    Input Low=Constant()
    Input High=Constant() {r=1}
    Target Low=ScalarParameter() {defaultValue=0.005, parameterName=Thickness_Low}
    Target High=ScalarParameter() {defaultValue=0.03, parameterName=Thickness_High}
   ) {materialFunction=&RemapValueRange}.Result)
  n13 = ScalarParameter() {defaultValue=0.75, parameterName=Wave2Offset}
  n14 = Add(A=Multiply(A=@n3, B=Add(A=@n2, B=@n13) {constB=0.5}), B=NamedRerouteUsage())
  n15 = DDX(Value=@n14)
  n16 = DDY(Value=@n14)
  n17 = Multiply(A=Add(A=Abs(None=@n15), B=Abs(None=@n16))) {constB=0.5}
  n18 = ScalarParameter() {defaultValue=0.08, parameterName=LineDistance}
  n19 = Add(A=Divide(A=@n14, B=@n18)) {constB=0.5}
  n20 = Abs(None=Subtract(A=Frac(None=@n19)) {constB=0.5})
  n21 = OneMinus(None=SmoothStep(Min=Subtract(A=@n12, B=@n17), Max=Add(A=@n12, B=@n17), Value=Multiply(A=@n20, B=@n18)))
  Max(A=@n11, B=@n21)
  Multiply(A=Multiply(A=@n21, B=ParticleColor().A), B=VertexColor().A)
  n22 = LinearInterpolate(A=Constant(), B=VectorParameter() {parameterName=WaveColor2, defaultValue=<rgba 0.0122 0.0049 0.0052 1>}.RGB, Alpha=@n21) {constA=0.18}
  n23 = VectorParameter() {parameterName=WaveColor1, defaultValue=<rgba 0.12 1.2 0.912 1>}
  n24 = TextureSample() {texture=/BRCosmetics/Gadgets/Assets/VinderTech_Umbrella/Umbrella_Season26/Textures/T_Umbrella_Season_26_FX}
  Multiply(
   A=Multiply(A=LinearInterpolate(A=@n22, B=@n23.RGB, Alpha=@n11) {constA=0.18}, B=@n24.R)
   B=ScalarParameter() {defaultValue=0.6, parameterName=EmissiveIntensity}
  )
  n25 = Multiply(A=@n24.R, B=Max(A=@n24.G, B=ScalarParameter() {defaultValue=0.5, parameterName=ScreenMarginBrightness}))
  Multiply(A=@n11, B=@n25)
  Multiply(A=@n25, B=@n21)
  n26 = TextureCoordinate()
  n27 = Constant2Vector()
  n28 = ScalarParameter() {defaultValue=2, parameterName=SDFSize}
  n29 = TextureSample(UVs=MaterialFunctionCall(UVs=Add(A=@n26, B=@n27), Texture Scale=@n28) {materialFunction=&ScaleUVsByCenter}.UVs) {
   samplerType=SAMPLERTYPE_LinearColor
   texture=/BRCosmetics/Gadgets/Assets/VinderTech_Umbrella/Umbrella_Season26/Textures/T_Umbrella_Season_26_SDF
  }
  NamedRerouteDeclaration(None=@n29.R) {name=SDF1, nodeColor=<rgba 0 1 0.8125 1>}
  n30 = ScalarParameter() {defaultValue=1, parameterName=ThicknessPower}
  NamedRerouteDeclaration(None=Power(Base=@n29.G, Exp=@n30)) {name=Thickness1, nodeColor=<rgba 0.7344 1 0 1>}
  n31 = Add(A=TextureCoordinate(), B=Constant2Vector())
  n32 = TextureSample(UVs=MaterialFunctionCall(
    UVs=MaterialFunctionCall(UVs=@n31, Texture Scale=@n28) {materialFunction=&ScaleUVsByCenter}.UVs
    "Rotation Angle (0-1)"=ScalarParameter() {defaultValue=0.225, parameterName=SDF2Rotation}
   ) {materialFunction=&CustomRotator}.Rotated Values) {
   samplerType=SAMPLERTYPE_LinearColor
   texture=&T_Umbrella_Season_26_SDF
  }
  NamedRerouteDeclaration(None=@n32.R) {name=SDF2, nodeColor=<rgba 0 1 0.8125 1>}
  NamedRerouteDeclaration(None=Power(Base=@n32.G, Exp=@n30)) {name=Thickness2, nodeColor=<rgba 0.7344 1 0 1>}
 ]
}
