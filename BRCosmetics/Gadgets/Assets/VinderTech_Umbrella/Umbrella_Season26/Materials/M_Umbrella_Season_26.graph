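# Season 26 VinderTech umbrella glider material that adds animated two-color waves to the character shading stack.
# Two offset samples of an SDF texture, one rotated, with offsets flipped across a diagonal UV split, supply distance and thickness reroutes.
# Four MF_S26Wave calls, maxed in pairs, tint base color through FX-texture screen and margin masks.
# The masked result also drives emissive, set between CharacterSkinFX and the PostFX functions.
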Material /BRCosmetics/Gadgets/Assets/VinderTech_Umbrella/Umbrella_Season26/Materials/M_Umbrella_Season_26
graph {
 properties {
  bAutomaticallySetUsageInEditor=false
  bUseMaterialAttributes=true
  bUsedWithClothing=true
  bUsedWithMeshDeformer=true
  bUsedWithMorphTargets=true
  bUsedWithNanite=true
  bUsedWithSkeletalMesh=true
  pixelDepthOffsetMode=PDOM_Legacy
  previewMesh=/Game/Gadgets/Assets/VinderTech_Umbrella/Umbrella_Season26/Meshes/Season_26_Glider
  refractionMethod=RM_IndexOfRefraction
  shadingModel=MSM_FromMaterialExpression
  subsurfaceProfile=&SS_HeroSkin_02
  usedShadingModels=Default Lit | Subsurface | Clear Coat | Subsurface Profile | Cloth | Eye
 }
 nodes [
  Constant() {r=1}
  n1 = MaterialFunctionCall() {materialFunction=&BaseFN_CharacterFunction_PreFX}
  n2 = MaterialFunctionCall(
   PreskinnedPosition=@n1.Pre skinned Local Position
   Gmap=@n1.Pre skinned Local Position
   MaterialAttributes=@n1.Pre skinned Local Position
   AO=@n1.Pre skinned Local Position
  ) {materialFunction=&CharacterSkinFX}
  n3 = MaterialFunctionCall(Float2=MaterialFunctionCall!(
    In=TextureSample() {texture=/BRCosmetics/Gadgets/Assets/VinderTech_Umbrella/Umbrella_Season26/Textures/T_Umbrella_Season_26_FX}.RGB
    Contrast=Constant() {r=1}
   ) {materialFunction=/Engine/Functions/Engine_MaterialFunctions01/ImageAdjustment/CheapContrast_RGB}.Result) {materialFunction=&BreakOutFloat2Components}
  n4 = Multiply(A=@n3.R, B=Max(A=@n3.G, B=ScalarParameter() {defaultValue=0.5, parameterName=ScreenMarginBrightness2}))
  n5 = ScalarParameter() {defaultValue=0.2, parameterName=LineDistance}
  n6 = ScalarParameter() {defaultValue=-0.05, parameterName=Speed}
  n7 = ScalarParameter() {defaultValue=1.5, parameterName=Wave2Offset}
  n8 = ScalarParameter() {defaultValue=0.007, parameterName=Thickness_Low}
  n9 = ScalarParameter() {defaultValue=0.02, parameterName=Thickness_High}
  n10 = ScalarParameter() {defaultValue=0.25, parameterName=Thickness_Variance}
  n11 = ScalarParameter() {defaultValue=2, parameterName=WaveAlternationOffset}
  n12 = Max(
   A=MaterialFunctionCall!(
    SDF=NamedRerouteUsage()
    WaveFrequency=@n5
    WaveSpeed=@n6
    WaveOffset=@n7
    Thickness=NamedRerouteUsage()
    WaveThickness_Low=@n8
    WaveThickness_High=@n9
    WaveThickness_Variance=@n10
   ) {materialFunction=&/MF_S26Wave}.Result
   B=MaterialFunctionCall(
    SDF=NamedRerouteUsage()
    WaveFrequency=@n5
    WaveSpeed=@n6
    WaveOffset=Add(A=@n11, B=@n7)
    Thickness=NamedRerouteUsage()
    WaveThickness_Low=@n8
    WaveThickness_High=@n9
    WaveThickness_Variance=@n10
   ) {materialFunction=&MF_S26Wave}.Result
  )
  n13 = MaterialFunctionCall(
   SDF=NamedRerouteUsage()
   WaveFrequency=@n5
   WaveSpeed=@n6
   WaveOffset=Constant()
   Thickness=NamedRerouteUsage()
   WaveThickness_Low=@n8
   WaveThickness_High=@n9
   WaveThickness_Variance=@n10
  ) {materialFunction=&MF_S26Wave}
  n14 = MaterialFunctionCall(
   SDF=NamedRerouteUsage()
   WaveFrequency=@n5
   WaveSpeed=@n6
   WaveOffset=@n11
   Thickness=NamedRerouteUsage()
   WaveThickness_Low=@n8
   WaveThickness_High=@n9
   WaveThickness_Variance=@n10
  ) {materialFunction=&MF_S26Wave}
  n15 = Max(A=@n3.G, B=ScalarParameter() {defaultValue=0.5, parameterName=ScreenMarginBrightness1})
  n16 = LinearInterpolate(
   A=LinearInterpolate(
    A=NamedRerouteUsage()
    B=VectorParameter() {parameterName=WaveColor2, defaultValue=<rgba 0.0122 0.0049 0.0052 1>}.RGB
    Alpha=Multiply(A=@n4, B=@n12)
   ) {constA=0.18}
   B=VectorParameter() {parameterName=WaveColor1, defaultValue=<rgba 0.0232 0.3613 0.2542 1>}.RGB
   Alpha=Multiply(A=Max(A=@n13.Result, B=@n14.Result), B=Multiply(A=@n3.R, B=@n15))
  ) {constA=0.18}
  n17 = Multiply(A=@n16, B=@n3.R)
  n18 = ScalarParameter() {defaultValue=0.6, parameterName=EmissiveIntensity}
  n19 = MaterialFunctionCall(
   AO=@n1.AO
   Distance Blend=@n1.AO
   Fresnel=@n1.AO
   Gmap=@n1.AO
   Pre Skinned Local Position=@n1.AO
   Material=SetMaterialAttributes!(MaterialAttributes=@n2.Result, Base Color=@n16, Emissive Color=Multiply(A=@n17, B=@n18)) {attributeSetTypes [#1, #2]}
  ) {materialFunction=&BaseFN_CharacterFunction_PostFX}
  MaterialFunctionCall(In=@n19.Result) {materialFunction=&CINE_MF_CharPostFX}
  n20 = TextureCoordinate()
  n21 = NamedRerouteUsage()
  n22 = MaterialFunctionCall(UVs=TextureCoordinate(), "Rotation Angle (0-1)"=Constant() {r=0.125}) {materialFunction=&CustomRotator}
  n23 = Ceil(None=Subtract(A=ComponentMask(None=@n22.Rotated Values) {r=true}) {constB=0.5})
  n24 = Subtract(A=@n21, B=Multiply(A=Multiply(A=@n23, B=@n21)) {constB=2})
  n25 = Multiply(A=Constant2Vector() {g=-0.12, r=0.18}, B=@n24)
  n26 = ScalarParameter() {defaultValue=0.76, parameterName=SDFSize}
  n27 = TextureSample(UVs=MaterialFunctionCall(UVs=Add(A=@n20, B=@n25), Texture Scale=@n26) {materialFunction=&ScaleUVsByCenter}.UVs) {
   samplerType=SAMPLERTYPE_LinearColor
   texture=/BRCosmetics/Gadgets/Assets/VinderTech_Umbrella/Umbrella_Season26/Textures/T_Umbrella_Season_26_SDF
  }
  NamedRerouteDeclaration(None=@n27.R) {name=SDF1, nodeColor=<rgba 0 1 0.8125 1>}
  n28 = ScalarParameter() {defaultValue=1, parameterName=ThicknessPower}
  NamedRerouteDeclaration(None=Power(Base=@n27.G, Exp=@n28)) {name=Thickness1, nodeColor=<rgba 0.7344 1 0 1>}
  n29 = Add(A=TextureCoordinate(), B=Multiply(A=Constant2Vector() {g=-0.3, r=0.25}, B=@n24) {constB=-1})
  n30 = TextureSample(UVs=MaterialFunctionCall(
    UVs=MaterialFunctionCall(UVs=@n29, Texture Scale=@n26) {materialFunction=&ScaleUVsByCenter}.UVs
    "Rotation Angle (0-1)"=ScalarParameter() {defaultValue=0.24, parameterName=SDF2Rotation}
   ) {materialFunction=&CustomRotator}.Rotated Values) {
   samplerType=SAMPLERTYPE_LinearColor
   texture=&T_Umbrella_Season_26_SDF
  }
  NamedRerouteDeclaration(None=@n30.A) {name=Thickness2_Offset, nodeColor=<rgba 0.1663 1 0 1>}
  NamedRerouteDeclaration(None=@n30.R) {name=SDF2, nodeColor=<rgba 0 1 0.8125 1>}
  NamedRerouteDeclaration(None=@n3.R) {name=UmbrellaMask, nodeColor=<rgba 1 0 0.5859 1>}
  NamedRerouteDeclaration(None=Power(Base=@n30.G, Exp=@n28)) {name=Thickness2, nodeColor=<rgba 0.7344 1 0 1>}
  n31 = GetMaterialAttributes(None=@n2.Result) {attributeGetTypes [#1, #3, #4]}
  NamedRerouteDeclaration(None=@n31.Base Color) {name=InBaseColor, nodeColor=<rgba 0 0.4844 1 1>}
  NamedRerouteDeclaration(None=@n27.B) {name=SDF1_Offset, nodeColor=<rgba 1 0.3984 0 1>}
  NamedRerouteDeclaration(None=@n31.Specular) {name=InSpecular, nodeColor=<rgba 1 0 0.1875 1>}
  NamedRerouteDeclaration(None=@n27.A) {name=Thickness1_Offset, nodeColor=<rgba 0.1663 1 0 1>}
  NamedRerouteDeclaration(None=@n30.B) {name=SDF2_Offset, nodeColor=<rgba 1 0.3984 0 1>}
 ]
}
